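# Animation Blueprint for the LeafyVest pickaxe mesh, holding the weapon in its idle pose.
# A single-state Idle state machine plays the idle pose sequence, then passes through DefaultSlot so montages can override it before the output pose.
# A Constraint and Transform Bone chain on the two-handed handle bone is left unconnected to the output.
# The event graph contains only unwired default nodes.
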AnimBlueprint /BRCosmetics/Weapons/FORT_Melee/Pickaxe_LeafyVest/Animation/LeafyVest_Pickaxe_BP
target_skeleton=/BRCosmetics/Weapons/FORT_Melee/Pickaxe_LeafyVest/Meshes/LeafyVest_Axe_Skeleton
graphs [
 AnimGraph {nodes [
   "Transform (Modify) Bone - Bone: Melee_TwoHanded_Handle"(
    ComponentPose="Constraint - Melee_TwoHanded_Handle ()"!(ComponentPose, Alpha=1) -> (Pose) {alphaBoolBlend {blendOption=Linear}, boneToModify {boneName=Melee_TwoHanded_Handle}}
    Alpha=1
    Translation="0,0,0"
    Rotation="0,-90,5"
    Scale="1,1,1"
   ) {rotationMode=BMM_Replace, alphaBoolBlend {blendOption=Linear}, boneToModify {boneName=Melee_TwoHanded_Handle}, rotation {roll=5, yaw=-90}}
   AnimGraphNode_StateMachine_0 = Idle_State() {is=title AnimGraphNode_StateMachine}
   Output Pose(Result=Slot 'DefaultSlot'(Source=@AnimGraphNode_StateMachine_0)) {layerGroup=None}
  ]}
 Idle_State {entry=Idle, path=AnimGraph.AnimGraphNode_StateMachine_0.Idle_State, ids=id, nodes [AnimStateNode_0 = Idle()]}
 Idle {
  path=AnimGraph.AnimGraphNode_StateMachine_0.Idle_State.AnimStateNode_0.Idle
  nodes [Output Animation Pose(Result=Sequence Player 'LeafyVest_Idle_Pose_WEP'() {sequence=&/LeafyVest_Idle_Pose_WEP})]
 }
 EventGraph {nodes [
   TryGetPawnOwner()
   Event BlueprintUpdateAnimation()
  ]}
]
variables []
